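# Idle Niagara effect for the Accumulate pickaxe, producing intermittent drips from bones or sockets on the user-supplied skeletal mesh.
# Drip1001 adds small drifting sprites with curl noise and inherited source movement.
# Drip2 and Drip3 re-sample their attachment points each frame, sliding dual-renderer sprites along an offset as they stretch and fade.
# Hero_Drips swings mesh drips on spring-driven pendulums, passing pivot, position and rigidity to dynamic material parameters.
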NiagaraSystem /BRCosmetics/Weapons/FORT_Melee/Pickaxe_Accumulate_Male/FX/NS_Pickaxe_Accumulate_Idle
dependencies {
 usedDataInterfaces [&NiagaraDataInterfaceCurve, &NiagaraDataInterfaceVector2DCurve, &NiagaraDataInterfaceColorCurve]
 usedDynamicInputs []
 usedModules [&SystemState]
}
emitters {
 Drip1001 {
  emitterUpdateScript [
   &EmitterState(
    Life Cycle Mode=::NewEnumerator1
    Loop Delay=&UniformRangedFloat(Minimum=0.8, Maximum=1.5)
    Loop Duration=1
    UseLoopDelay=false
   )
   &SpawnRate(Spawn Probability=0.45, SpawnRate=&RandomRangeFloat(Minimum=1, Maximum=1), Use Spawn Probability=true)
  ]
  particleSpawnScript [
   /Niagara/Modules/Spawn/Initialization/InitializeParticle(
    Lifetime=&UniformRangedFloat(Minimum=1, Maximum=2)
    Mass=&UniformRangedFloat(Minimum=0.75, Maximum=1.75)
    Module.Write Color=true
    Module.Write Lifetime=true
    Module.Write Mass=false
    Module.Write Material Random=true
    Module.Write Position=false
    Module.Write Scale=false
    Module.Write SpriteRotation=true
    Module.Write SpriteSize=true
    Position=&SimulationPosition()
    Sprite Rotation=90
    Sprite Size=&Vector2DFromFloat(Value=&RandomRangeFloat(Minimum=2, Maximum=3.5))
   )
   &SkeletalMeshLocation(
    Apply Sampled Orientation to Mesh Particles=true
    Inherit Velocity=false
    MaskSpawnGroup=false
    Mesh Sampling Type=::NewEnumerator5
    Sampled Position Offset=<0 0 6>
    Sampled Position Offset Coordinate Space:ENiagaraMeshTransforms=::NewEnumerator3
    Skeletal Mesh=$User.SkeletalMesh
    Socket Sampling Mode=::NewEnumerator5
   ) {module=SkeletalMeshLocation001}
   &AddVelocity(
    Coordinate Space=::Simulation
    Velocity=&RotateVectorByQuaternion(
     Vector To Rotate=&RandomRangeVector(Minimum=<0 0 5>, Maximum=<0 0 10>)
     Quaternion=$Output.SkeletalMeshLocation001.SampledRotation
    )
   )
  ]
  particleUpdateScript [
   &InheritSourceMovement(
    Applied Position Falloff Distance=50
    Applied Position Scale=<0.6 0.6 0.6>
    ApplyForce=false
    ApplyPosition=true
    ApplyPositionFalloff=true
    ApplyPositionFalloffCurve=true
   )
   &CurlNoiseForce(Noise Frequency=3, Pan Noise=true, Pan Noise Field=<0 0 0.5>)
   &SolveForcesAndVelocity(Clamp Velocity=false, Limit Acceleration=false)
   &ParticleState()
   &CameraOffset(Camera Offset Amount=&RandomRangeFloat(Minimum=5, Maximum=8))
   &Color(Color=&ColorFromCurve(ColorCurve=DataInterface() {propertyValues {
       AlphaCurve {keys [<at 0.0106 0>, <at 0.1528 1>, {arriveTangent=-1.2022, interpMode=RCIM_Cubic, leaveTangent=-1.2022, time=0.5807, value=1}, <at 0.9846 0>]}
       BlueCurve {keys [<at 0.0774 1 RCIM_Cubic>,<at 1 1 RCIM_Cubic>]}
       GreenCurve {keys [<at 0.0774 1 RCIM_Cubic>,<at 1 1 RCIM_Cubic>]}
       RedCurve {keys [<at 0.0774 1 RCIM_Cubic>,<at 1 1 RCIM_Cubic>]}
      }}))
   &ScaleSpriteSize(Scale Factor=&Vector2DFromFloat(Value=&FloatFromCurve(FloatCurve=DataInterface() {propertyValues {Curve {keys [
          {arriveTangent=9.694, interpMode=RCIM_Cubic, leaveTangent=9.694, tangentMode=RCTM_User, time=0, value=0}
          <at 0.2 1 RCIM_Cubic>
          {arriveTangent=-2.6965, interpMode=RCIM_Cubic, leaveTangent=-2.6965, tangentMode=RCTM_User, time=1, value=0}
         ]}}})))
  ]
  rendererClasses [&NiagaraSpriteRendererProperties]
  renderers [{rendererClass=&NiagaraSpriteRendererProperties}]
 }
 Drip2 {
  emitterUpdateScript [
   &EmitterState(
    Life Cycle Mode=::NewEnumerator1
    Loop Delay=&UniformRangedFloat(Minimum=0.8, Maximum=1.5)
    Loop Duration=1
    UseLoopDelay=false
   )
   &SpawnRate(Spawn Probability=0.6, SpawnRate=&RandomRangeFloat(Minimum=1, Maximum=2), Use Spawn Probability=true)
  ]
  particleSpawnScript [
   &InitializeParticle(
    Lifetime=&UniformRangedFloat(Minimum=0.8, Maximum=2.5)
    Mass=&UniformRangedFloat(Minimum=0.75, Maximum=1.75)
    Module.Write Color=true
    Module.Write Lifetime=true
    Module.Write Mass=false
    Module.Write Material Random=true
    Module.Write Position=false
    Module.Write Scale=false
    Module.Write SpriteRotation=true
    Module.Write SpriteSize=true
    Position=&SimulationPosition()
    Sprite Rotation=90
    Sprite Size=&Multiply_Vector2(
     A=&RandomRangeVector2D(Minimum=<0.8 0.8>, Maximum=<1.2 1.2>)
     B=&Vector2DFromFloat(Value=&RandomRangeFloat(Minimum=2, Maximum=2.5))
    )
   )
   /SetVariables(Particles.DripSize:Vector2f=$Particles.SpriteSize) {isSetParameters=true, module=SetVariables_#1}
   /SetVariables(Particles.StartPositionOffeset:Vector3f=&RandomRangeVector(Minimum=<-0.2 -0.5 -1>, Maximum=<0.2 0.5 -2>)) {isSetParameters=true, module=SetVariables_#2}
   &SkeletalMeshLocation(
    Apply Sampled Orientation to Mesh Particles=true
    Bone / Socket Index=1
    Inherit Velocity=false
    MaskSpawnGroup=false
    Mesh Sampling Type=::NewEnumerator5
    Sampled Position Offset=$Particles.StartPositionOffeset
    Sampled Position Offset Coordinate Space:ENiagaraMeshTransforms=::NewEnumerator3
    Skeletal Mesh=$User.SkeletalMesh
    Socket Sampling Mode=::NewEnumerator5
   ) {module=SkeletalMeshLocation001}
   &AddVelocity(Velocity=&RotateVectorByQuaternion(Vector To Rotate=<0 0 200>, Quaternion=$Output.SkeletalMeshLocation001.SampledRotation))
  ]
  particleUpdateScript [
   &SolveForcesAndVelocity(Clamp Velocity=false, Limit Acceleration=false)
   &ParticleState()
   &SkeletalMeshLocation(
    Apply Sampled Orientation to Mesh Particles=true
    Bone / Socket Index=1
    Inherit Velocity=false
    MaskSpawnGroup=false
    Mesh Sampling Type=::NewEnumerator5
    Sampled Position Offset=&Lerp_Vector(
     A=$Particles.Initial.StartPositionOffeset
     B=&Multiply_VectorByFloat(
      Vector=<0 0 1>
      Float=&FloatFromCurve(
       FloatCurve=DataInterface() {propertyValues {Curve {keys [
           <at 0 0>
           {arriveTangent=0.6747, interpMode=RCIM_Cubic, leaveTangent=0.6747, tangentMode=RCTM_User, time=0.6242, value=0.1558}
           {arriveTangent=5.2544, interpMode=RCIM_Cubic, leaveTangent=5.2544, tangentMode=RCTM_User, time=0.997, value=1}
          ]}}}
       Scale Curve=&RandomRangeFloat(Minimum=8, Maximum=12)
      )
     )
     Alpha=$Particles.NormalizedAge
    )
    Sampled Position Offset Coordinate Space:ENiagaraMeshTransforms=::NewEnumerator3
    Skeletal Mesh=$User.SkeletalMesh
    Socket Sampling Mode=::NewEnumerator5
   )
   /SetVariables(Particles.DripSize:Vector2f=&Multiply_Vector2(
     A=$Particles.Initial.DripSize
     B=&Vector2DFromCurve(Vector2Curve=DataInterface() {propertyValues {XCurve {keys [<at 0 1>,<at 1 1>]}, YCurve {keys [<at 0 4>,<at 1 1>]}}}, ScaleCurve=<1 1>)
    )) {isSetParameters=true, module=SetVariables_#3}
   /SetVariables(Particles.DripColor:LinearColor=&ColorFromCurve(ColorCurve=DataInterface() {propertyValues {
       AlphaCurve {keys [<at 0.5723 0 RCIM_Cubic>,<at 0.6433 1 RCIM_Cubic>]}
       BlueCurve {keys [<at 0 1 RCIM_Cubic>,<at 1 1 RCIM_Cubic>]}
       GreenCurve {keys [<at 0 1 RCIM_Cubic>,<at 1 1 RCIM_Cubic>]}
       RedCurve {keys [<at 0 1 RCIM_Cubic>,<at 1 1 RCIM_Cubic>]}
      }})) {isSetParameters=true, module=SetVariables_#4}
   &Color(Color=&ColorFromCurve(ColorCurve=DataInterface() {propertyValues {
       AlphaCurve {keys [
         <at 0.0106 0>
         <at 0.069 1>
         {arriveTangent=-1.657, interpMode=RCIM_Cubic, leaveTangent=-1.657, time=0.5807, value=1}
         <at 0.6725 0 RCIM_Cubic>
        ]}
       BlueCurve {keys [<at 0.0774 1 RCIM_Cubic>,<at 1 1 RCIM_Cubic>]}
       GreenCurve {keys [<at 0.0774 1 RCIM_Cubic>,<at 1 1 RCIM_Cubic>]}
       RedCurve {keys [<at 0.0774 1 RCIM_Cubic>,<at 1 1 RCIM_Cubic>]}
      }}))
   &ScaleSpriteSize(Scale Factor=&Vector2DFromCurve(
     Vector2Curve=DataInterface() {propertyValues {XCurve {keys [<at 0 1>,<at 0.7052 0.4662>]}, YCurve {keys [<at -0.0001 0.4>,<at 0.6407 1.5>]}}}
     ScaleCurve=<1 1>
    )) {module=ScaleSpriteSize001}
   &CameraOffset(Camera Offset Amount=&RandomRangeFloat(Minimum=2, Maximum=5)) {module=CameraOffset001}
  ]
  rendererClasses [&NiagaraSpriteRendererProperties]
  renderers [
   {rendererClass=&NiagaraSpriteRendererProperties}
   {rendererClass=&NiagaraSpriteRendererProperties, rendererIndex=1}
  ]
 }
 Drip3 {
  emitterUpdateScript [
   &EmitterState(
    Life Cycle Mode=::NewEnumerator1
    Loop Delay=&UniformRangedFloat(Minimum=0.8, Maximum=1.5)
    Loop Duration=1
    UseLoopDelay=false
   )
   &SpawnRate(Spawn Probability=0.6, SpawnRate=&RandomRangeFloat(Minimum=1, Maximum=2), Use Spawn Probability=true)
  ]
  particleSpawnScript [
   &InitializeParticle(
    Lifetime=&UniformRangedFloat(Minimum=0.8, Maximum=2.5)
    Mass=&UniformRangedFloat(Minimum=0.75, Maximum=1.75)
    Module.Write Color=true
    Module.Write Lifetime=true
    Module.Write Mass=false
    Module.Write Material Random=true
    Module.Write Position=false
    Module.Write Scale=false
    Module.Write SpriteRotation=true
    Module.Write SpriteSize=true
    Position=&SimulationPosition()
    Sprite Rotation=90
    Sprite Size=&Multiply_Vector2(
     A=&RandomRangeVector2D(Minimum=<0.8 0.8>, Maximum=<1.2 1.2>)
     B=&Vector2DFromFloat(Value=&RandomRangeFloat(Minimum=1.5, Maximum=2))
    )
   )
   /SetVariables(Particles.DripSize:Vector2f=$Particles.SpriteSize) {isSetParameters=true, module=SetVariables_#1}
   /SetVariables(Particles.StartPositionOffeset:Vector3f=&RandomRangeVector(Minimum=<0 0 -0.5>, Maximum=<0 0 -1>)) {isSetParameters=true, module=SetVariables_#2}
   &SkeletalMeshLocation(
    Apply Sampled Orientation to Mesh Particles=true
    Bone / Socket Index=2
    Inherit Velocity=false
    MaskSpawnGroup=false
    Mesh Sampling Type=::NewEnumerator5
    Sampled Position Offset=$Particles.StartPositionOffeset
    Sampled Position Offset Coordinate Space:ENiagaraMeshTransforms=::NewEnumerator3
    Skeletal Mesh=$User.SkeletalMesh
    Socket Sampling Mode=::NewEnumerator5
   ) {module=SkeletalMeshLocation001}
   &AddVelocity(Velocity=&RotateVectorByQuaternion(Vector To Rotate=<0 0 200>, Quaternion=$Output.SkeletalMeshLocation001.SampledRotation))
  ]
  particleUpdateScript [
   &SolveForcesAndVelocity(Clamp Velocity=false, Limit Acceleration=false)
   &ParticleState()
   &SkeletalMeshLocation(
    Apply Sampled Orientation to Mesh Particles=true
    Bone / Socket Index=2
    Inherit Velocity=false
    MaskSpawnGroup=false
    Mesh Sampling Type=::NewEnumerator5
    Sampled Position Offset=&Lerp_Vector(
     A=$Particles.Initial.StartPositionOffeset
     B=&Multiply_VectorByFloat(
      Vector=<0 0 1>
      Float=&FloatFromCurve(
       FloatCurve=DataInterface() {propertyValues {Curve {keys [
           <at 0 0>
           {arriveTangent=0.6747, interpMode=RCIM_Cubic, leaveTangent=0.6747, tangentMode=RCTM_User, time=0.6242, value=0.1558}
           {arriveTangent=5.2544, interpMode=RCIM_Cubic, leaveTangent=5.2544, tangentMode=RCTM_User, time=0.997, value=1}
          ]}}}
       Scale Curve=&RandomRangeFloat(Minimum=5, Maximum=10)
      )
     )
     Alpha=$Particles.NormalizedAge
    )
    Sampled Position Offset Coordinate Space:ENiagaraMeshTransforms=::NewEnumerator3
    Skeletal Mesh=$User.SkeletalMesh
    Socket Sampling Mode=::NewEnumerator5
   )
   /SetVariables(Particles.DripSize:Vector2f=&Multiply_Vector2(
     A=$Particles.Initial.DripSize
     B=&Vector2DFromCurve(Vector2Curve=DataInterface() {propertyValues {XCurve {keys [<at 0 1>,<at 1 1>]}, YCurve {keys [<at 0 4>,<at 1 1>]}}}, ScaleCurve=<1 1>)
    )) {isSetParameters=true, module=SetVariables_#3}
   /SetVariables(Particles.DripColor:LinearColor=&ColorFromCurve(ColorCurve=DataInterface() {propertyValues {
       AlphaCurve {keys [<at 0.5723 0 RCIM_Cubic>,<at 0.6433 1 RCIM_Cubic>]}
       BlueCurve {keys [<at 0 1 RCIM_Cubic>,<at 1 1 RCIM_Cubic>]}
       GreenCurve {keys [<at 0 1 RCIM_Cubic>,<at 1 1 RCIM_Cubic>]}
       RedCurve {keys [<at 0 1 RCIM_Cubic>,<at 1 1 RCIM_Cubic>]}
      }})) {isSetParameters=true, module=SetVariables_#4}
   &Color(Color=&ColorFromCurve(ColorCurve=DataInterface() {propertyValues {
       AlphaCurve {keys [
         <at 0.0106 0>
         <at 0.069 1>
         {arriveTangent=-1.657, interpMode=RCIM_Cubic, leaveTangent=-1.657, time=0.5807, value=1}
         <at 0.6725 0 RCIM_Cubic>
        ]}
       BlueCurve {keys [<at 0.0774 1 RCIM_Cubic>,<at 1 1 RCIM_Cubic>]}
       GreenCurve {keys [<at 0.0774 1 RCIM_Cubic>,<at 1 1 RCIM_Cubic>]}
       RedCurve {keys [<at 0.0774 1 RCIM_Cubic>,<at 1 1 RCIM_Cubic>]}
      }}))
   &ScaleSpriteSize(Scale Factor=&Vector2DFromCurve(
     Vector2Curve=DataInterface() {propertyValues {XCurve {keys [<at 0 1>,<at 0.7052 0.4662>]}, YCurve {keys [<at -0.0001 0.4>,<at 0.6407 1.5>]}}}
     ScaleCurve=<1 1>
    )) {module=ScaleSpriteSize001}
   &CameraOffset(Camera Offset Amount=&RandomRangeFloat(Minimum=2, Maximum=5))
  ]
  rendererClasses [&NiagaraSpriteRendererProperties]
  renderers [
   {rendererClass=&NiagaraSpriteRendererProperties}
   {rendererClass=&NiagaraSpriteRendererProperties, rendererIndex=1}
  ]
 }
 Hero_Drips {
  emitterUpdateScript [
   &EmitterState()
   &SpawnRate(Spawn Probability=0.75, SpawnRate=0.5, Use Spawn Probability=true)
  ]
  particleSpawnScript [
   /SetVariables(
    Particles.PendulumDrag=1
    Particles.PendulumLengthScale:NiagaraFloat=&RandomRangeFloat(Minimum=0.35, Maximum=0.45)
    Particles.PendulumMeshWidth:NiagaraFloat=&RandomRangeFloat(Minimum=0.05, Maximum=0.1)
    Particles.PendulumRigidity=0.06
   ) {isSetParameters=true}
   &InitializeParticle(
    Lifetime=&UniformRangedFloat(Minimum=3, Maximum=4)
    Mesh Scale=&MakeVector(X=$Particles.PendulumMeshWidth, Y=$Particles.PendulumMeshWidth, Z=$Particles.PendulumLengthScale)
    Module.Write Color=true
    Module.Write Lifetime=true
    Module.Write Mass=true
    Module.Write Material Random=true
    Module.Write Position=false
    Module.Write Scale=true
    Module.Write SpriteRotation=false
    Module.Write SpriteSize=true
   )
   &CameraOffset(Camera Offset Amount=1)
  ]
  particleUpdateScript [
   &ParticleState()
   &SkeletalMeshLocation(
    Apply Sampled Orientation to Mesh Particles=true
    Apply To Particle Position=false
    Inherit Velocity=false
    MaskSpawnGroup=false
    Mesh Sampling Type=::NewEnumerator5
    Skeletal Mesh=$User.SkeletalMesh
    Socket Sampling Mode=::NewEnumerator5
   )
   /SetVariables(Particles.MeshOrigin:Vector3f=&Add_Vector(
     A=$Output.SkeletalMeshLocation.SampledPosition
     B=&RotateVectorByQuaternion(Vector To Rotate=<-1 0 -3>, Quaternion=$Output.SkeletalMeshLocation.SampledRotation)
    )) {isSetParameters=true}
   &Drag(Drag=$Particles.PendulumDrag, Use Linear Drag=true, Use Rotational Drag=false)
   &GravityForce()
   &PendulumSetup(
    Module.Enable Spring Driver=true
    Pendulum Length:NiagaraFloat=&Multiply_Float(A=100, B=$Particles.PendulumLengthScale)
    Pendulum Pivot:Vector3f=$Particles.MeshOrigin
    Pendulum Rest Axis:Vector3f=&RotateVectorByQuaternion(Vector To Rotate=<0 0 1>, Quaternion=$Output.SkeletalMeshLocation.SampledRotation)
    "Spring Driven Constraint (Tightness)"=10
   )
   /SetVariables(Particles.MeshFacing:Vector3f=&NormalizeVector(Vector To Normalize=&Subtract_Vector(A=$Particles.MeshOrigin, B=Linked() {linkedVariable {name=Particles.Position, type=NiagaraPosition}}))) {isSetParameters=true}
   &self:Pendulum_Facing(
    Bone Quat:Quat4f=$Output.SkeletalMeshLocation.SampledRotation
    FromVector:Vector3f=<0 0 -1>
    ToVector:Vector3f=$Particles.MeshFacing
    X Vector:Vector3f=<-1 0 0>
    Y Vector:Vector3f=<0 0 0>
   ) {module=Pendulum_Facing}
   &SolveForcesAndVelocity(Clamp Velocity=false, Limit Acceleration=false)
   &PendulumConstraint()
   &CalculateAccurateVelocity()
   &DynamicMaterialParameters(
    Index 1 Param 3=1
    Index 1 Param 4=&FloatFromCurve(FloatCurve=DataInterface() {propertyValues {Curve {keys [
         {arriveTangent=3.6645, interpMode=RCIM_Cubic, leaveTangent=3.6645, tangentMode=RCTM_User, time=0.1, value=-0.1}
         {interpMode=RCIM_Cubic, leaveTangent=2.3573, time=0.9, value=2}
        ]}}})
    Index 2 Param 1=HlslExpression() {hlslExpression=Particles.MeshOrigin.x}
    Index 2 Param 2=HlslExpression() {hlslExpression=Particles.MeshOrigin.y}
    Index 2 Param 3=HlslExpression() {hlslExpression=Particles.MeshOrigin.z}
    Index 2 Param 4=$Particles.PendulumMaskFalloff
    Index 3 Param 1=&MakeFloatFromVector(Vector=Linked() {linkedVariable {name=Particles.Position, type=NiagaraPosition}}, Channel=::NewEnumerator0)
    Index 3 Param 2=&MakeFloatFromVector(Vector=Linked() {linkedVariable {name=Particles.Position, type=NiagaraPosition}}, Channel=::NewEnumerator1)
    Index 3 Param 3=&MakeFloatFromVector(Vector=Linked() {linkedVariable {name=Particles.Position, type=NiagaraPosition}}, Channel=::NewEnumerator2)
    Index 3 Param 4=$Particles.PendulumRigidity
    Module.Param0WriteEnabled1=true
    Module.Param0WriteEnabled2=true
    Module.Param0WriteEnabled3=true
    Module.Param1WriteEnabled1=true
    Module.Param1WriteEnabled2=true
    Module.Param1WriteEnabled3=true
    Module.Param2WriteEnabled1=true
    Module.Param2WriteEnabled2=true
    Module.Param2WriteEnabled3=true
    Module.Param3WriteEnabled1=true
    Module.Param3WriteEnabled2=true
    Module.Param3WriteEnabled3=true
    Write Parameter Index 0=false
    Write Parameter Index 1=true
    Write Parameter Index 2=true
    Write Parameter Index 3=true
   )
   &ScaleMeshSize(Scale Factor=&MakeVector(
     X=&FloatFromCurve(FloatCurve=DataInterface() {propertyValues {Curve {keys [
          <at 0 0 RCIM_Cubic>
          {arriveTangent=2.5, interpMode=RCIM_Cubic, leaveTangent=2.5, time=0.05, value=0.9}
          {arriveTangent=-0.7353, interpMode=RCIM_Cubic, leaveTangent=-0.7353, time=0.6, value=0.65}
          {arriveTangent=-4.2857, interpMode=RCIM_Cubic, leaveTangent=-4.2857, time=0.9, value=0.375}
          {interpMode=RCIM_Cubic, leaveTangent=-0.6579, time=0.95, value=0}
         ]}}})
     Y=&FloatFromCurve(FloatCurve=DataInterface() {propertyValues {Curve {keys [
          <at 0 0 RCIM_Cubic>
          {arriveTangent=2.5, interpMode=RCIM_Cubic, leaveTangent=2.5, time=0.05, value=0.9}
          {arriveTangent=-0.7353, interpMode=RCIM_Cubic, leaveTangent=-0.7353, time=0.6, value=0.65}
          {arriveTangent=-4.2857, interpMode=RCIM_Cubic, leaveTangent=-4.2857, time=0.9, value=0.375}
          {interpMode=RCIM_Cubic, leaveTangent=-0.6579, time=0.95, value=0}
         ]}}})
     Z=&FloatFromCurve(FloatCurve=DataInterface() {propertyValues {Curve {keys [
          {arriveTangent=1.8796, interpMode=RCIM_Cubic, time=0, value=0}
          {arriveTangent=-4.7929, interpMode=RCIM_Cubic, leaveTangent=-4.7929, tangentMode=RCTM_User, time=0.625, value=0.85}
          {arriveTangent=-1.7696, interpMode=RCIM_Cubic, leaveTangent=-1.7696, tangentMode=RCTM_User, time=0.95, value=0}
         ]}}})
    ))
  ]
  rendererClasses [&NiagaraMeshRendererProperties]
  renderers [{rendererClass=&NiagaraMeshRendererProperties, rendererIndex=0}]
 }
}
summary {
 systemName=NS_Pickaxe_Accumulate_Idle
 emitters [
  {emitterName=Drip2, rendererClasses [&NiagaraSpriteRendererProperties]}
  {emitterName=Drip3, rendererClasses [&NiagaraSpriteRendererProperties]}
  {emitterName=Hero_Drips, rendererClasses [&NiagaraMeshRendererProperties]}
  {emitterName=Drip1001, rendererClasses [&NiagaraSpriteRendererProperties]}
 ]
 userVariables [
  User.SkeletalMesh {
   type=NiagaraDataInterfaceSkeletalMesh
   defaultValue {
    struct=NiagaraExt_VariableValue_DataInterface
    value {dataInterface=&self:NiagaraDataInterfaceSkeletalMesh_1}
   }
  }
  User.SlurpColorAlpha {type=NiagaraFloat, defaultValue=0, description=""}
 ]
}
